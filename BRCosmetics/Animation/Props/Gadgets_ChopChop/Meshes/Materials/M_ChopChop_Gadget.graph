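# Character material for the ChopChop gadget prop that switches blink, blush and mouth poses by collapsing mesh pieces.
# Scalar pose parameters pick target values compared by narrow sphere masks against vertex color R (blink), B (blush) and G (mouth).
# World Position Offset pulls unmatched vertices to the object origin, while vertices with RGB summing above one are exempt.
# This offset is injected between the shared 2023 character parent PreFX and PostFX functions.
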Material /BRCosmetics/Animation/Props/Gadgets_ChopChop/Meshes/Materials/M_ChopChop_Gadget
graph {
 properties {
  bUseMaterialAttributes=true
  bUsedWithMeshDeformer=true
  bUsedWithMorphTargets=true
  bUsedWithSkeletalMesh=true
  pixelDepthOffsetMode=PDOM_Legacy
  previewMesh=/BRCosmetics/Animation/Props/Gadgets_ChopChop/Meshes/Demo/ChopChop
  shadingModel=MSM_FromMaterialExpression
  usedShadingModels=Default Lit | Subsurface | Clear Coat | Subsurface Profile | Cloth
 }
 nodes [
  n1 = MaterialFunctionCall() {materialFunction=&MF_CharacterParent_2023_PreFX}
  n2 = Multiply(
   A=Subtract(A=ObjectPositionWS() {originType=CameraRelative}, B=WorldPosition() {worldPositionShaderOffset=WPT_CameraRelativeNoOffsets}.XYZ)
   B=NamedRerouteUsage()
  )
  MaterialFunctionCall(
   Material=SetMaterialAttributes!(
    MaterialAttributes=@n1.Result
    World Position Offset=NamedRerouteDeclaration(None=@n2) {name=WPO for Collapsing, nodeColor=<rgba 1 0 0.1406 1>}
   ) {attributeSetTypes [#1]}
   M=@n1.M
   Pre Skinned Local Position=@n1.M
   Distance Blend=@n1.M
   Fresnel=@n1.M
  ) {materialFunction=&MF_CharacterParent_2023_PostFX}
  MaterialExpressionConstant_0 = Constant() {r=0.1}
  MaterialExpressionConstant_1 = Constant() {r=0.2}
  n3 = Switch!(
   SwitchValue=ScalarParameter() {group=ChopChop, parameterName=blink_pose}
   Default=@MaterialExpressionConstant_0
   0=@MaterialExpressionConstant_0
   1=@MaterialExpressionConstant_1
  ) {
   constDefault=0.1
   description=Mouth Pose Switch
   inputs [
    {
     inputName=0
     input {
      expression=&self:MaterialExpressionConstant_0
      inputName=None
      mask=0
      maskA=0
      maskB=0
      maskG=0
      maskR=0
      outputIndex=0
     }
    }
    {
     inputName=1
     input {
      expression=&self:MaterialExpressionConstant_1
      inputName=None
      mask=0
      maskA=0
      maskB=0
      maskG=0
      maskR=0
      outputIndex=0
     }
    }
   ]
  }
  n4 = OneMinus(None=SphereMask(A=@n3, B=VertexColor().R) {attenuationRadius=0.025})
  MaterialExpressionConstant_4 = Constant() {r=0.5}
  MaterialExpressionConstant_3 = Constant() {r=0.1}
  n5 = Switch(
   SwitchValue=ScalarParameter() {defaultValue=1, group=ChopChop, parameterName=blush_pose}
   Default=@MaterialExpressionConstant_4
   0=@MaterialExpressionConstant_4
   1=@MaterialExpressionConstant_3
  ) {
   description=Mouth Pose Switch
   inputs [
    {
     inputName=0
     input {
      expression=&self:MaterialExpressionConstant_4
      inputName=None
      mask=0
      maskA=0
      maskB=0
      maskG=0
      maskR=0
      outputIndex=0
     }
    }
    {
     inputName=1
     input {
      expression=&self:MaterialExpressionConstant_3
      inputName=None
      mask=0
      maskA=0
      maskB=0
      maskG=0
      maskR=0
      outputIndex=0
     }
    }
   ]
  }
  n6 = VertexColor()
  MaterialExpressionConstant_9 = Constant() {r=0.1}
  MaterialExpressionConstant_8 = Constant() {r=0.2}
  MaterialExpressionConstant_10 = Constant() {r=0.3}
  n7 = Switch!(
   SwitchValue=ScalarParameter() {group=ChopChop, parameterName=mouth_pose}
   Default=@MaterialExpressionConstant_9
   0=@MaterialExpressionConstant_9
   1=@MaterialExpressionConstant_8
   2=@MaterialExpressionConstant_10
  ) {
   description=Mouth Pose Switch
   inputs [
    {
     inputName=0
     input {
      expression=&self:MaterialExpressionConstant_9
      inputName=None
      mask=0
      maskA=0
      maskB=0
      maskG=0
      maskR=0
      outputIndex=0
     }
    }
    {
     inputName=1
     input {
      expression=&self:MaterialExpressionConstant_8
      inputName=None
      mask=0
      maskA=0
      maskB=0
      maskG=0
      maskR=0
      outputIndex=0
     }
    }
    {
     inputName=2
     input {
      expression=&self:MaterialExpressionConstant_10
      inputName=None
      mask=0
      maskA=0
      maskB=0
      maskG=0
      maskR=0
      outputIndex=0
     }
    }
   ]
  }
  n8 = VertexColor()
  n9 = Multiply(A=OneMinus(None=SphereMask(A=@n5, B=@n6.B) {attenuationRadius=0.025}), B=OneMinus(None=SphereMask(A=@n7, B=@n8.G) {attenuationRadius=0.025}))
  n10 = DotProduct(A=VertexColor(), B=Constant3Vector() {constant=<rgba 1 1 1 0>})
  n11 = Constant()
  n12 = Constant() {r=1}
  NamedRerouteDeclaration(None=Multiply(A=Multiply(A=@n4, B=@n9), B=If(A=@n10, "A > B"=@n11, "A < B"=@n12) {constB=1})) {name=WPO Opacity Mask, nodeColor=<rgba 0 0.4609 1 1>}
 ]
}
